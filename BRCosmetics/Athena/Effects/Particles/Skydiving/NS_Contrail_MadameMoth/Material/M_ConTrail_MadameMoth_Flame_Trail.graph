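# Translucent, unlit, two-sided flame trail material for the Madame Moth skydiving contrail, used on Niagara ribbons and mesh particles.
# World position offset combines SplineThicken width, a masked sine wave and an optional particle-position pull, driven by dynamic parameters.
# Opacity carves panned toon-flame textures with noise dissolves, then applies end and depth fades.
# Emissive multiplies curve-atlas color by panned noise, particle color and time-of-day tint, with per-particle speed and U offset.
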Material /BRCosmetics/Athena/Effects/Particles/Skydiving/NS_Contrail_MadameMoth/Material/M_ConTrail_MadameMoth_Flame_Trail
graph {
 properties {
  bUsedWithNiagaraMeshParticles=true
  bUsedWithNiagaraRibbons=true
  blendMode=BLEND_Translucent
  pixelDepthOffsetMode=PDOM_Legacy
  shadingModel=MSM_Unlit
  twoSided=true
  usedShadingModels=Unlit
 }
 nodes [
  n1 = DynamicParameter() {defaultValue=<rgba 1 1 1 0>, paramNames [WPO, Speed, WPO Disturbance, U offset]}
  n2 = ComponentMask(None=TextureCoordinate()) {r=true}
  n3 = ScalarParameter() {parameterName=SplineThicken Edge Strenght}
  n4 = Saturate(None=Multiply(A=LinearInterpolate(A=@n2, Alpha=@n3), B=ScalarParameter() {defaultValue=3, parameterName=SplineTicken Edge Contrast}) {constB=5})
  n5 = Multiply(A=Multiply(A=ScalarParameter() {defaultValue=10, parameterName=SplineThickenBase}, B=@n1.WPO), B=@n4)
  n6 = Multiply(A=Multiply(A=ScalarParameter() {defaultValue=40, parameterName=SplineThickenTip}, B=@n1.WPO), B=@n4)
  n7 = StaticBool()
  n8 = DynamicParameter() {defaultValue=<rgba 1 1 1 0>, paramNames [WPO, Speed, WPO Disturbance, U offset]}
  n9 = Constant()
  n10 = ComponentMask(None=Add(A=TextureCoordinate() {uTiling=2.2}, B=AppendVector(A=@n8.U offset, B=@n9))) {r=true}
  n11 = Multiply(A=Time()) {constB=-1.3}
  n12 = Multiply(A=Sine(None=Add(A=@n10, B=@n11)) {period=1.2}, B=VectorParameter() {parameterName=WPO Direction, defaultValue=<rgba 0 0 1 1>}.RGB)
  n13 = Multiply(
   A=LinearInterpolate(
    A=Transform(None=@n12) {transformSourceType=TRANSFORMSOURCE_World, transformType=TRANSFORM_Camera}
    B=Transform(None=@n12) {transformSourceType=TRANSFORMSOURCE_Local}
    Alpha=ScalarParameter() {parameterName=Change Transform}
   )
   B=DynamicParameter() {defaultValue=<rgba 1 1 1 0>, paramNames [WPO, Speed, WPO Disturbance, U offset]}.WPO Disturbance
  )
  n14 = ComponentMask(None=TextureCoordinate()) {r=true}
  n15 = LinearInterpolate(Alpha=Saturate(None=Multiply(A=@n14) {constB=3}))
  n16 = DynamicParameter() {parameterIndex=1, defaultValue=<rgba 1 1 1 0>, paramNames [X,Y,Z,WPOIntensity]}
  n17 = Clamp(
   None=Subtract(
    A=MaterialFunctionCall(X=@n16.X, Y=@n16.X, Z=@n16.X) {materialFunction=&MakeFloat3}.Result
    B=ParticlePositionWS()
   )
   Min=Constant() {r=-30}
   Max=Constant() {r=30}
  )
  n18 = TextureCoordinate()
  n19 = Multiply(A=@n16.WPOIntensity, B=Power(Base=ComponentMask(None=@n18) {r=true}, Exp=ScalarParameter() {defaultValue=2, parameterName=WPOBlendPower}))
  Add(
   A=Add(
    A=MaterialFunctionCall(WidthBase=@n5, WidthTip=@n6, Expand U or V UV Channel=@n7) {materialFunction=&SplineThicken}.WorldPositionOffset
    B=Multiply(A=@n13, B=@n15)
   ) {constB=0}
   B=LinearInterpolate(B=LinearInterpolate(B=@n17, Alpha=@n19), Alpha=ScalarParameter() {parameterName=Use Interpolate WPO})
  )
  Constant()
  Constant()
  n20 = DynamicParameter() {defaultValue=<rgba 1 1 1 0>, paramNames [WPO, Speed, WPO Disturbance, U offset]}
  n21 = Constant()
  n22 = Panner(
   Coordinate=Add(A=TextureCoordinate() {uTiling=3, vTiling=2}, B=AppendVector(A=@n20.U offset, B=@n21))
   Time=DynamicParameter() {defaultValue=<rgba 1 1 1 0>, paramNames [WPO, Speed, WPO Disturbance, U offset]}.Speed
  ) {speedX=-0.5}
  n23 = Saturate(None=LinearInterpolate(Alpha=TextureSample(UVs=@n22) {samplerType=SAMPLERTYPE_Grayscale, texture=&LowResBlurredNoise}.R) {constB=1.5})
  n24 = Multiply(A=TextureCoordinate(), B=ComponentMask(None=VectorParameter() {parameterName=UV Scale, defaultValue=<rgba 0.8 1 0 0>}.RGB) {g=true, r=true})
  n25 = AppendVector(A=Power(Base=ComponentMask(None=@n24) {r=true}) {constExponent=0.7}, B=ComponentMask(None=@n24) {g=true})
  n26 = AppendVector(A=DynamicParameter() {defaultValue=<rgba 1 1 1 0>, paramNames [WPO, Speed, WPO Disturbance, U offset]}.U offset, B=Constant())
  n27 = TextureSampleParameter2D(UVs=Panner(
    Coordinate=Add(A=@n25, B=@n26)
    Time=DynamicParameter() {defaultValue=<rgba 1 1 1 0>, paramNames [WPO, Speed, WPO Disturbance, U offset]}.Speed
   ) {speedX=-0.3}) {parameterName=Tex_Default, texture=/Game/Effects/Niagara/MasterMaterials/SplineThickenTrail/Texture/T_FlameToon_Trail}
  n28 = Multiply(A=TextureCoordinate(), B=ComponentMask(None=VectorParameter() {parameterName=UV Scale, defaultValue=<rgba 0.8 1 0 0>}.RGB) {g=true, r=true})
  n29 = AppendVector(A=Power(Base=ComponentMask(None=@n28) {r=true}) {constExponent=0.7}, B=ComponentMask(None=@n28) {g=true})
  n30 = AppendVector(A=DynamicParameter() {defaultValue=<rgba 1 1 1 0>, paramNames [WPO, Speed, WPO Disturbance, U offset]}.U offset, B=Constant())
  n31 = TextureSampleParameter2D(UVs=Panner(
    Coordinate=Add(A=@n29, B=@n30)
    Time=DynamicParameter() {defaultValue=<rgba 1 1 1 0>, paramNames [WPO, Speed, WPO Disturbance, U offset]}.Speed
   ) {speedX=-0.3}) {parameterName=Tex_Default, texture=&T_FlameToon_Trail}
  n32 = Subtract(
   A=Power(
    Base=ChannelMaskParameter(None=@n31.RGBA) {group=Tex, parameterName=Mask}.R
    Exp=ScalarParameter() {defaultValue=1, group=Tex, parameterName=Tex alpha fix}
   ) {constExponent=2.2}
   B=LinearInterpolate(B=Multiply(A=@n27.B, B=@n23), Alpha=ScalarParameter() {defaultValue=1, parameterName=Lerp Center Dissolve})
  )
  n33 = ComponentMask(None=TextureCoordinate()) {r=true}
  n34 = Saturate(None=Multiply(A=OneMinus(None=@n33)) {constB=5})
  n35 = ParticleColor()
  n36 = ScalarParameter() {defaultValue=-2, parameterName=Disappear min}
  n37 = ScalarParameter() {defaultValue=2, parameterName=Disappear max}
  n38 = VectorParameter() {parameterName=Bright Noise Scale, defaultValue=<rgba 1 1 0 0>}
  n39 = DynamicParameter() {defaultValue=<rgba 1 1 1 0>, paramNames [WPO, Speed, WPO Disturbance, U offset]}
  n40 = Add(
   A=Multiply(A=TextureCoordinate() {uTiling=2, vTiling=0.8}, B=ComponentMask(None=@n38.RGB) {g=true, r=true})
   B=AppendVector(A=@n39.U offset, B=Constant())
  )
  n41 = ScalarParameter() {parameterName=Dissolve mask Min}
  n42 = ScalarParameter() {defaultValue=0.5, parameterName=Dissolve mask Max}
  n43 = Power(Base=ComponentMask(None=TextureCoordinate()) {r=true}) {constExponent=1.3}
  n44 = Power(
   Base=TextureSample(UVs=Panner(Coordinate=@n40, Time=@n39.Speed) {speedX=-0.5}) {samplerType=SAMPLERTYPE_LinearGrayscale, texture=/Game/Global/Textures/Default/Caustics/T_Nosie_Caustics_Grayscale}.R
   Exp=Saturate(None=LinearInterpolate(A=@n41, B=@n42, Alpha=@n43) {constA=0.1, constB=2})
  )
  n45 = Multiply(A=Multiply(A=Multiply(A=@n32, B=@n34), B=@n35.A), B=Saturate(None=LinearInterpolate(A=@n36, B=@n37, Alpha=@n44) {constA=-2, constB=2}))
  n46 = TextureCoordinate()
  n47 = Constant2Vector() {g=0.2, r=1}
  n48 = LinearInterpolate(B=OneMinus(None=SphereMask(A=@n46, B=@n47) {attenuationRadius=0.5, hardnessPercent=50}), Alpha=ScalarParameter() {parameterName=Fixed tips}) {constA=1}
  n49 = Multiply(A=ComponentMask(None=TextureCoordinate()) {r=true}, B=ScalarParameter() {defaultValue=10, parameterName=Fade_Base}) {constB=10}
  DepthFade(
   Opacity=Multiply(A=Multiply(A=@n45, B=@n48), B=Saturate(None=@n49))
   FadeDistance=ScalarParameter() {defaultValue=10, parameterName=FadeDistance}
  )
  n50 = Fresnel(ExponentIn=ScalarParameter() {defaultValue=1, parameterName=FresnelExponent})
  LinearInterpolate(B=Multiply(B=OneMinus(None=@n50)), Alpha=ScalarParameter() {parameterName=Use Fresnel})
  n51 = Constant() {r=1}
  n52 = VectorParameter() {parameterName="Color_Night (V3)", defaultValue=<rgba 1.5 1.5 1.5 1>}
  n53 = Multiply(
   A=TextureCoordinate() {uTiling=2}
   B=ComponentMask(None=VectorParameter() {parameterName=Dark Noise UV Scale, defaultValue=<rgba 1 1 0 0>}.RGB) {g=true, r=true}
  )
  n54 = AppendVector(A=DynamicParameter() {defaultValue=<rgba 1 1 1 0>, paramNames [WPO, Speed, WPO Disturbance, U offset]}.U offset, B=Constant())
  n55 = TextureSample(UVs=Panner(
    Coordinate=Add(A=@n53, B=@n54)
    Time=DynamicParameter() {defaultValue=<rgba 1 1 1 0>, paramNames [WPO, Speed, WPO Disturbance, U offset]}.Speed
   ) {speedX=-1}) {samplerType=SAMPLERTYPE_LinearGrayscale, texture=/Engine/Functions/Engine_MaterialFunctions02/ExampleContent/Textures/LowResBlurredNoise}
  n56 = DynamicParameter() {parameterIndex=2, defaultValue=<rgba 1 1 1 0>, paramNames [Add Noise Lerp, none, none, none]}
  n57 = ScalarParameter() {parameterName=Lerp Dy Noise}
  n58 = LinearInterpolate(
   B=Clamp(None=LinearInterpolate(Alpha=@n55.R) {constA=-0.8}, Min=ScalarParameter() {defaultValue=0.1, parameterName=Dark noise min}) {minDefault=0.1}
   Alpha=Multiply(A=ScalarParameter() {defaultValue=1, parameterName=Add Noise}, B=LinearInterpolate(B=@n56.Add Noise Lerp, Alpha=@n57) {constA=1})
  ) {constA=1}
  n59 = TextureCoordinate()
  n60 = CurveAtlasRowParameter(CurveTime=LinearInterpolate(A=ComponentMask(None=@n59) {r=true}, B=@n58, Alpha=ScalarParameter() {parameterName=UseNoiseAsCurveTime})) {
   atlas=/Game/Athena/LinearColorCurves/CAtlas_Contrail_MadameMoth_Color
   curve=/Game/Athena/LinearColorCurves/CV_Contrail_MadameMoth_Color1
   parameterName=Color
  }
  n61 = LinearInterpolate(
   B=LinearInterpolate(A=Multiply(A=@n58, B=@n60), B=@n60, Alpha=ScalarParameter() {parameterName=UseNoiseAsCurveTime})
   Alpha=ScalarParameter() {defaultValue=1, parameterName=Use Color Curve}
  ) {constA=1}
  MaterialFunctionCall(Emissive=Multiply(
    A=MaterialFunctionCall(Color_Morning=@n51, Color_Day=@n51, Color_Evening=@n51, Color_Night=@n52.RGB) {materialFunction=&Color_Time_of_Day}.Result
    B=Multiply(A=@n61, B=@n35.RGB)
   )) {materialFunction=&MF_FX_EmissiveAdjustments}
  Round(None=@n44)
 ]
}
