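# Emerald variant of the MadameMoth skydiving contrail, attaching green fire, sparks, glows and mesh trails to the character's skeleton.
# GPU fire and spark sprites rise from sampled bones with curl noise, while persistent burst-spawned particles carry mesh trails, rings, and sprite-plus-light glows pinned to bones.
# ParaGlideLeanAlpha drives trail scale, material animation speed, velocity and turbulence.
# Front-end, time-of-day and platform modifiers tune spawn rates, sizes and brightness.
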NiagaraSystem /BRCosmetics/Athena/Effects/Particles/Skydiving/NS_Contrail_MadameMoth/NS_ConTrail_MadameMoth_Emerald
dependencies {
 usedDataInterfaces [
  &NiagaraDataInterfaceArrayFloat3
  &NiagaraDataInterfacePlatformSet
  &NiagaraDataInterfaceArrayInt32
  &NiagaraDataInterfaceSkeletalMesh
  &NiagaraDataInterfaceArrayFloat
  &NiagaraDataInterfaceCurve
  &NiagaraDataInterfaceColorCurve
 ]
 usedDynamicInputs []
 usedModules [
  &self:SystemSpawnScript.SystemScriptSource.SystemScriptGraph.NiagaraNodeAssignment_0.SetVariables_#1
  &self:SystemSpawnScript.SystemScriptSource.SystemScriptGraph.NiagaraNodeAssignment_1.SetVariables_#2
  &self:SystemSpawnScript.SystemScriptSource.SystemScriptGraph.NiagaraNodeAssignment_2.SetVariables_#3
  &self:SystemSpawnScript.SystemScriptSource.SystemScriptGraph.NiagaraNodeAssignment_3.SetVariables_#4
  &SystemState
 ]
}
emitters {
 Fire001 {
  simTarget=GPUComputeSim
  emitterUpdateScript [
   &EmitterState()
   &SpawnRate(SpawnRate=&FN_FrontEndFloatModifier(Gameplay=30, Front End=15), Use Spawn Probability=false)
  ]
  particleSpawnScript [
   &PartitionParticles(Number Of Partitions=4)
   /SetVariables(Particles.Random size:NiagaraFloat=&RandomRangeFloat(Minimum=0.2, Maximum=0.25)) {isSetParameters=true}
   &InitializeParticle(
    Color=<rgba 0 5 1.6667 1>
    Lifetime Max=0.6
    Lifetime Min=0.4
    Lifetime Mode=::NewEnumerator1
    Sprite Rotation Mode=::NewEnumerator2
    Sprite Size Mode=::NewEnumerator3
    Sprite UV Mode=::NewEnumerator3
    Uniform Sprite Size=&Multiply_Float(A=100, B=$Particles.Random size)
    UsePositionOffset=false
    Write Material Random=true
   ) {module=InitializeParticle001}
   &DynamicMaterialParameters(
    Index 0 Param 2=&RandomRangeFloat(Minimum=0, Maximum=4)
    Index 0 Param 4=2.5
    Module.Param0WriteEnabled=true
    Module.Param1WriteEnabled=true
    Module.Param2WriteEnabled=false
    Module.Param3WriteEnabled=true
   ) {module=DynamicMaterialParameters001}
   &SkeletalMeshLocation(
    Bone / Socket Index=$Particles.PartitionParticles.Partition
    Bone Sampling Mode=::NewEnumerator4
    MaskSpawnGroup=false
    Position Sampling=::NewEnumerator1
    Skeletal Mesh=$System.Skeletal Mesh
   )
   &AddVelocity(
    Rotation Coordinate Space=::World
    Velocity=<0 0 100>
    Velocity Speed Scale=&Lerp_Float(A=2, B=4, Alpha=&ClampFloat(Float=$User.ParaGlideLeanAlpha))
   ) {module=AddVelocity001}
   &CameraOffset(Camera Offset Amount=-50) {module=CameraOffset001}
   /SetVariables() {isSetParameters=true}
   /SetVariables(Particles.DY3:Vector4f=&MakeVector4(X=3, Y=1, Z=1, W=1)) {isSetParameters=true}
  ]
  particleUpdateScript [
   /SetVariables(Particles.Size Scale:NiagaraFloat=&FloatFromCurve(FloatCurve=DataInterface() {propertyValues {Curve {keys [<at 0 0>,<at 0.1 0.5>,<at 1 0>]}}})) {isSetParameters=true}
   &SubUVAnimation(End Frame=11, Start Frame=8, SubUV Animation Mode:ENiagara_SubUVLookupMode=::NewEnumerator2)
   /Niagara/Modules/Update/Forces/CurlNoiseForce(
    Noise Frequency=100
    Noise Strength=&Multiply_Float(
     A=2000
     B=&Lerp_Float(A=0.5, B=1.5, Alpha=&ClampFloat(Float=$User.ParaGlideLeanAlpha))
    )
    Pan Noise=true
   )
   &ParticleState()
   &ScaleSpriteSize(Scale Sprite Size Mode=::NewEnumerator0, Uniform Scale Factor=$Particles.Size Scale) {module=ScaleSpriteSize001}
   &SpriteRotationRate(Rotation Rate=&RandomRangeFloat(Minimum=-5, Maximum=5)) {module=SpriteRotationRate001}
   &Drag(Drag=4, Use Linear Drag=true, Use Rotational Drag=true)
   &SolveForcesAndVelocity(Clamp Velocity=false, Limit Acceleration=false)
   &ScaleColor(
    Scale Alpha=&Multiply_Float(
     A=&FloatFromCurve(FloatCurve=DataInterface() {propertyValues {Curve {keys [
          {interpMode=RCIM_Cubic, tangentMode=RCTM_User, time=0, value=0}
          {interpMode=RCIM_Cubic, tangentMode=RCTM_User, time=0.5, value=0.75}
          {tangentMode=RCTM_User, time=1, value=0}
         ]}}})
     B=1
    )
    Scale RGB=&Multiply_VectorByFloat(
     Vector=&MakeVectorFromLinearColorRGB(Color=&ColorFromCurve(ColorCurve=DataInterface() {propertyValues {
         AlphaCurve {keys []}
         BlueCurve {keys [<at 0 1 RCIM_Cubic>,<at 0.3533 0>,<at 0.6474 0>]}
         GreenCurve {keys [<at 0 1 RCIM_Cubic>,<at 0.3533 0>,<at 0.6474 0>]}
         RedCurve {keys [<at 0 1 RCIM_Cubic>,<at 0.3533 0.25>,<at 0.6474 0.1>]}
        }}))
     Float=0.5
    )
    ScaleA=true
    ScaleRGB=true
   )
   &FN_Sprite_Attributes(
    Color=&MakeLinearColorFromVectorAndFloat(
     "Vector (RGB)"=&MakeVectorFromLinearColorRGB(Color=&ColorFromCurve(ColorCurve=DataInterface() {propertyValues {
         AlphaCurve {keys []}
         BlueCurve {keys [<at 0 1 RCIM_Cubic>,<at 1 0.1667 RCIM_Cubic>]}
         GreenCurve {keys [<at 0 2 RCIM_Cubic>,<at 1 1 RCIM_Cubic>]}
         RedCurve {keys [<at 0 0 RCIM_Cubic>,<at 1 0 RCIM_Cubic>]}
        }}))
     "Float (Alpha)"=2
    )
    Create Color Attribute=true
    Create SpriteSize Attribute=true
    SpriteSize=&Vector2DFromFloat(Value=&Multiply_Float(A=60, B=&Multiply_Float(A=$Particles.Random size, B=$Particles.Size Scale)))
    Write to Module Namespace=true
   )
   &FN_Sprite_Attributes(
    Color=&MakeLinearColorFromVectorAndFloat(
     "Vector (RGB)"=&MakeVectorFromLinearColorRGB(Color=<rgba 0 1.5 0.75 0.5>)
     "Float (Alpha)"=&Multiply_Float(
      A=&FN_FrontEndFloatModifier(
       Gameplay=&Multiply_Float(A=0.2, B=&FNTimeOfDayScalar(Night Scale=0.5, Morning Scale=1, Evening Scale=1))
       Front End=0.1
      )
      B=&FloatFromCurve(
       FloatCurve=DataInterface() {propertyValues {Curve {keys [{interpMode=RCIM_Cubic, tangentMode=RCTM_User, time=0, value=1}, {interpMode=RCIM_Cubic, tangentMode=RCTM_User, time=1, value=1}]}}}
       Scale Curve=0.5
      )
     )
    )
    Create Color Attribute=true
    Create SpriteSize Attribute=true
    SpriteSize=&Vector2DFromFloat(Value=&Multiply_Float(A=200, B=&Multiply_Float(A=$Particles.Random size, B=$Particles.Size Scale)))
    Write to Module Namespace=true
   ) {module=FN_Sprite_Attributes001}
  ]
  rendererClasses [&NiagaraSpriteRendererProperties]
  renderers [
   {rendererClass=&NiagaraSpriteRendererProperties}
   {rendererClass=&NiagaraSpriteRendererProperties, rendererIndex=1}
   {rendererClass=&NiagaraSpriteRendererProperties, rendererIndex=2}
  ]
 }
 ParticleDecal {
  emitterUpdateScript [
   &EmitterState(
    Inactive Response=::NewEnumerator1
    Life Cycle Mode=::NewEnumerator1
    Loop Behavior=::NewEnumerator1
    Loop Delay=1.4
    Loop Duration Mode=::NewEnumerator1
    UseLoopDelay=false
   )
   &SpawnBurst_Instantaneous(Spawn Count=4, Use Spawn Probability=false, UseLoopCountLimit=false)
  ]
  particleSpawnScript [
   &FN_SetSpawnGroup()
   &InitializeParticle(
    Color=<rgba 0 1 0.5 1>
    Sprite Rotation Mode=::NewEnumerator2
    Sprite Size Mode=::NewEnumerator3
    Uniform Sprite Size=&SelectFloatFromArray(Direct Array Index=$Particles.SpawnGroup, Float Selection Array=DataInterface() {propertyValues {FloatData [70,70,60,60]}})
    UsePositionOffset=false
    Write Material Random=true
   ) {module=InitializeParticle001}
   &CameraOffset(Camera Offset Amount=10)
  ]
  particleUpdateScript [
   &SkeletalMeshLocation(
    Bone / Socket Index=$Particles.SpawnGroup
    Bone Sampling Mode=::NewEnumerator4
    MaskSpawnGroup=false
    Orientation Sampling=::NewEnumerator1
    Sampled Position Offset Coordinate Space:ENiagaraSkelMeshTransforms=::NewEnumerator1
    Skeletal Mesh=$System.Skeletal Mesh
   ) {module=SkeletalMeshLocation001}
   &ParticleState(Kill Particles When Lifetime Has Elapsed=false, Let Infinitely Lived Particles Die When Emitter Deactivates=true)
   &ScaleSpriteSize(Scale Factor=&Vector2DFromFloat(Value=&FloatFromCurve(FloatCurve=DataInterface() {propertyValues {Curve {keys [
          {interpMode=RCIM_Cubic, leaveTangent=5.3269, tangentMode=RCTM_Break, time=0, value=0}
          {arriveTangent=0.6, interpMode=RCIM_Cubic, leaveTangent=0.6, time=0.2, value=0.54}
          {arriveTangent=0.0611, interpMode=RCIM_Cubic, tangentMode=RCTM_Break, time=1, value=0.6}
         ]}}})))
   &FN_Light_Attributes(
    Color=<rgba 1.25 30 15.625 1>
    Create Color Attribute=true
    Create Exponent Attribute=true
    Create Radius Attribute=true
    Exponent=2
    Radius=70
    Scale Color=&FN_FrontEndFloatModifier(Gameplay=1, Front End=0.2)
   )
   &ScaleColor(Scale Alpha=&FN_FrontEndFloatModifier(Gameplay=1, Front End=1), ScaleA=true, ScaleRGB=true)
  ]
  rendererClasses [&NiagaraSpriteRendererProperties, &NiagaraLightRendererProperties]
  renderers [
   {rendererClass=&NiagaraSpriteRendererProperties}
   {rendererClass=&NiagaraLightRendererProperties, rendererIndex=1}
  ]
 }
 Sparks {
  simTarget=GPUComputeSim
  emitterUpdateScript [
   &EmitterState()
   &SpawnRate(SpawnRate=&FN_FrontEndFloatModifier(Gameplay=20, Front End=0), Use Spawn Probability=false)
  ]
  particleSpawnScript [
   &PartitionParticles(Number Of Partitions=4)
   &InitializeParticle(
    Color=<rgba 0.8237 1.5504 2 1>
    Lifetime Max=0.5
    Lifetime Min=0.4
    Lifetime Mode=::NewEnumerator1
    Sprite Size Max=<2 20>
    Sprite Size Min=<2 20>
    Sprite Size Mode=::NewEnumerator2
    UsePositionOffset=false
    Write Material Random=true
   )
   &SkeletalMeshLocation(
    Bone / Socket Index=$Particles.PartitionParticles.Partition
    Bone Sampling Mode=::NewEnumerator4
    MaskSpawnGroup=false
    Position Sampling=::NewEnumerator1
    Skeletal Mesh=$System.Skeletal Mesh
   ) {module=SkeletalMeshLocation001}
   &AddVelocity(
    Rotation Coordinate Space=::World
    Velocity=<0 0 100>
    Velocity Speed Scale=&Lerp_Float(A=2, B=3, Alpha=&ClampFloat(Float=$User.ParaGlideLeanAlpha))
   ) {module=AddVelocity001}
  ]
  particleUpdateScript [
   &ParticleState()
   &CurlNoiseForce(
    Noise Frequency=100
    Noise Strength=&Multiply_Float(A=2000, B=&Lerp_Float(A=0.5, B=1, Alpha=&ClampFloat(Float=$User.ParaGlideLeanAlpha)))
    Pan Noise=true
   )
   /SetVariables(Particles.Glow03_Color=&MakeLinearColorFromVectorAndFloat(
     "Vector (RGB)"=&MakeVectorFromLinearColorRGB(Color=&MultiplyLinearColor(Linear Color=<rgba 0 1.5 0.75 1>, Scale Factor=&MakeVector4_FromFloat(Float=0.5)))
     "Float (Alpha)"=&Multiply_Float(
      A=&Multiply_Float(A=0.3, B=&FNTimeOfDayScalar(Night Scale=0.5, Morning Scale=1, Evening Scale=1))
      B=1
     )
    )) {isSetParameters=true}
   /SetVariables(Particles.Glow03_Size=&Multiply_Vector2D_ByFloat(
     Float=&FloatFromCurve(
      FloatCurve=DataInterface() {propertyValues {Curve {keys [
          <at 0 0 RCIM_Cubic>
          <at 0.2 1.3 RCIM_Cubic>
          {arriveTangent=-0.1423, interpMode=RCIM_Cubic, leaveTangent=-0.1423, tangentMode=RCTM_User, time=1, value=0}
         ]}}}
      Scale Curve=10
     )
     Vector2D=<0.6 1.2>
    )) {isSetParameters=true}
   &ScaleColor(
    Scale Alpha=&FloatFromCurve(FloatCurve=DataInterface() {propertyValues {Curve {keys [<at 0.3 1>,<at 0.4 1>,<at 1 1>]}}})
    ScaleA=true
    ScaleRGB=true
   )
   &ScaleSpriteSize(Scale Factor=&Vector2DFromFloat(Value=&FloatFromCurve(
      FloatCurve=DataInterface() {propertyValues {Curve {keys [
          <at 0 1 RCIM_Cubic>
          {interpMode=RCIM_Cubic, leaveTangent=-3.0296, tangentMode=RCTM_Break, time=0.2, value=1.2}
          {arriveTangent=-1.8107, interpMode=RCIM_Cubic, leaveTangent=-1.8107, tangentMode=RCTM_User, time=1, value=0}
         ]}}}
      Scale Curve=&Multiply_Float(
       A=0.3
       B=&ClampFloat(
        Float=&FN_LOD_Distance_Remap(Start Distance=0, End Distance=500, Minimum Output=0, Maximum Output=1, Clamp Output=true)
        Min=0.2
       )
      )
     ))) {module=ScaleSpriteSize001}
   &Drag(Use Linear Drag=true, Use Rotational Drag=true)
   &SolveForcesAndVelocity(Clamp Velocity=false, Limit Acceleration=false)
  ]
  rendererClasses [&NiagaraSpriteRendererProperties]
  renderers [
   {rendererClass=&NiagaraSpriteRendererProperties}
   {rendererClass=&NiagaraSpriteRendererProperties, rendererIndex=1}
  ]
 }
 Trail_Ring {
  emitterUpdateScript [
   &EmitterState(
    Inactive Response=::NewEnumerator1
    Life Cycle Mode=::NewEnumerator1
    Loop Behavior=::NewEnumerator1
    Loop Duration Mode=::NewEnumerator1
    UseLoopDelay=false
   )
   &SpawnBurst_Instantaneous(Spawn Count=4, Use Spawn Probability=false, UseLoopCountLimit=false)
  ]
  particleSpawnScript [
   &PartitionParticles(Number Of Partitions=4)
   &InitializeParticle(Color=<rgba 0.3 0.3 0.3 1>, Mesh Scale=<0.7 1 0.7>, Mesh Scale Mode=::NewEnumerator4, UsePositionOffset=false, Write Material Random=true)
   &InitialMeshOrientation(
    Mesh Orientation Mode=::NewEnumerator2
    Rotation=&SelectVectorFromArray(
     Direct Array Index=$Particles.PartitionParticles.Partition
     Vector Selection Array=DataInterface() {propertyValues {FloatData [<0 0 0.25>, <0 0 -0.25>, <0 0 0.5>, <0 0 0.15>]}}
    )
    Use Rotation Vector=true
   )
   &InitialMeshOrientation(
    Mesh Orientation Mode=::NewEnumerator2
    Rotation=&SelectVectorFromArray(
     Direct Array Index=$Particles.PartitionParticles.Partition
     Vector Selection Array=DataInterface() {propertyValues {FloatData [<0 0 0.5>, <0 0 0.5>, <0 0 0.15>, <0 0 0>]}}
    )
    Use Rotation Vector=true
   ) {module=InitialMeshOrientation001}
   &CameraOffset()
  ]
  particleUpdateScript [
   &SkeletalMeshLocation(
    Bone / Socket Index=$Particles.PartitionParticles.Partition
    Bone Sampling Mode=::NewEnumerator4
    MaskSpawnGroup=false
    Sampled Position Offset=&SelectVectorFromArray(
     Direct Array Index=$Particles.PartitionParticles.Partition
     Vector Selection Array=DataInterface() {propertyValues {FloatData [<30 0 0>, <-30 0 0>, <10 -5 0>, <-10 5 0>]}}
    )
    Skeletal Mesh=DataInterface() {propertyValues {
      DefaultMesh=null
      ExcludeBoneName=root
      OptionalComponentClass=null
      PreviewMesh=&M_Med_Soldier_04_Preview
      SkinningMode=SkinOnTheFly
      SoftSourceActor=null
      SourceMode=Default
      UvSetIndex=0
      WholeMeshLOD=-1
      bExcludeBone=false
      bReadDeformedGeometry=true
      bRequireCurrentFrameData=true
      ComponentTags []
      FilteredBones [lowerarm_l,lowerarm_r,foot_l,foot_r]
      FilteredSockets []
      MeshUserParameter {parameter {name=None, typeDefHandle {registeredTypeIndex=1}, varData []}}
      SamplingRegions []
     }}
   )
   &ParticleState(Kill Particles When Lifetime Has Elapsed=false, Let Infinitely Lived Particles Die When Emitter Deactivates=true)
   &Timeline(
    Play Rate=&Lerp_Float(A=2, B=3, Alpha=&ClampFloat(Float=$User.ParaGlideLeanAlpha))
    Timeline End=10000000
   )
   &DynamicMaterialParameters(
    Index 0 Param 1=&SelectFloatFromArray(
     Direct Array Index=$Particles.PartitionParticles.Partition
     Float Selection Array=DataInterface() {propertyValues {FloatData [0.8,0.8,0.5,0.5]}}
    )
    Index 0 Param 2=$Particles.Timeline.Time
    Index 0 Param 3=1
    Index 2 Param 1=0.7
    Module.Index0Param0Write=true
    Module.Index0Param1Write=true
    Module.Index0Param2Write=true
    Module.Index0Param3Write=true
    Module.Index2Param0Write=true
    Module.Index2Param1Write=true
    Module.Index2Param2Write=true
    Module.Index2Param3Write=true
    Write Parameter Index 2=true
   )
   &ScaleMeshSize(Scale Factor=&Multiply_Vector(
     A=&SelectVectorFromArray(
      Direct Array Index=$Particles.PartitionParticles.Partition
      Vector Selection Array=DataInterface() {propertyValues {FloatData [<1 1 1>, <1 1 -1>, <1 1 1>, <1 1 -1>]}}
     )
     B=&Lerp_Vector(A=<1 1 1>, B=<1 1 1>, Alpha=&ClampFloat(Float=$User.ParaGlideLeanAlpha))
    ))
   &ScaleColor(
    Scale RGB=&Multiply_VectorByFloat(
     Vector=&MakeVectorFromLinearColorRGB(Color=&FNTimeOfDayColor(Night Color=<rgba 0.1444 0.6 0.3956 1>, Morning Color=<rgba 1 1 1 1>, Day Color=<rgba 1 1 1 1>, Evening Color=<rgba 1 1 1 1>))
     Float=&FN_FrontEndFloatModifier(
      Gameplay=&FN_PickFloatDependOnPlatform(TargetPlatformActive=1, TargetPlatformNoActive=0.8)
      Front End=1
     )
    )
    ScaleA=true
    ScaleRGB=true
   )
   /SetVariables(Particles.Color1:LinearColor=&MakeLinearColorFromVectorAndFloat(
     "Vector (RGB)"=&MakeVectorFromLinearColorRGB(Color=&MultiplyLinearColor(
       Linear Color=<rgba 1 1 1 1>
       Scale Factor=&MakeVector4_FromFloat(Float=&FN_FrontEndFloatModifier(
         Gameplay=&Multiply_Float(
          A=&Multiply_Float(
           A=$System.Color1 TOD fix
           B=&FNTimeOfDayScalar(Night Scale=0.2, Morning Scale=1, Evening Scale=1)
          )
          B=&FN_PickFloatDependOnPlatform(TargetPlatformActive=1, TargetPlatformNoActive=0.7)
         )
         Front End=0.7
        ))
      ))
     "Float (Alpha)"=&FN_FrontEndFloatModifier(Gameplay=5, Front End=2)
    )) {isSetParameters=true, module=SetVariables_#5}
   /SetVariables(Particles.Color2=&MakeLinearColorFromVectorAndFloat(
     "Vector (RGB)"=&MakeVectorFromLinearColorRGB(Color=&MultiplyLinearColor(
       Linear Color=&FNTimeOfDayColor(Night Color=<rgba 3 3 3 1>, Morning Color=<rgba 1 1 1 1>, Day Color=<rgba 1 1 1 1>, Evening Color=<rgba 1 1 1 1>)
       Scale Factor=&MakeVector4_FromFloat(Float=&FN_FrontEndFloatModifier(Gameplay=$System.Color2 TOD fix, Front End=1.1))
      ))
     "Float (Alpha)"=0.35
    )) {isSetParameters=true, module=SetVariables_#6}
   /SetVariables(Particles.Color3:LinearColor=&MakeLinearColorFromVectorAndFloat(
     "Vector (RGB)"=&MakeVectorFromLinearColorRGB(Color=<rgba 0.5 0.5 0.5 1>)
     "Float (Alpha)"=&FN_FrontEndFloatModifier(Gameplay=0.3, Front End=0.3)
    )) {isSetParameters=true, module=SetVariables_#7}
  ]
  rendererClasses [&NiagaraMeshRendererProperties]
  renderers [
   {rendererClass=&NiagaraMeshRendererProperties, rendererIndex=0}
   {rendererClass=&NiagaraMeshRendererProperties, rendererIndex=1}
   {rendererClass=&NiagaraMeshRendererProperties, rendererIndex=2}
   {rendererClass=&NiagaraMeshRendererProperties, rendererIndex=3}
  ]
 }
 Trail_feet {
  emitterUpdateScript [
   &EmitterState(
    Inactive Response=::NewEnumerator1
    Life Cycle Mode=::NewEnumerator1
    Loop Behavior=::NewEnumerator1
    Loop Duration Mode=::NewEnumerator1
    UseLoopDelay=false
   )
   &SpawnBurst_Instantaneous(Spawn Count=2, Use Spawn Probability=false, UseLoopCountLimit=false)
  ]
  particleSpawnScript [
   &PartitionParticles()
   &InitializeParticle(
    Color=<rgba 0.4 0.4 0.4 1>
    Mesh Scale=&SelectVectorFromArray(
     Direct Array Index=$Particles.PartitionParticles.Partition
     Vector Selection Array=DataInterface() {propertyValues {FloatData [<0.5 0.5 0.4>, <0.5 -0.5 0.5>]}}
    )
    Mesh Scale Mode=::NewEnumerator4
    UsePositionOffset=false
    Write Material Random=true
   )
   &InitialMeshOrientation(
    Mesh Orientation Mode=::NewEnumerator2
    Rotation=&SelectVectorFromArray(
     Direct Array Index=$Particles.PartitionParticles.Partition
     Vector Selection Array=DataInterface() {propertyValues {FloatData [<0 0 0.3>, <0 0 0>]}}
    )
    Use Rotation Vector=true
   )
   &InitialMeshOrientation(
    Mesh Orientation Mode=::NewEnumerator2
    Rotation=&SelectVectorFromArray(
     Direct Array Index=$Particles.PartitionParticles.Partition
     Vector Selection Array=DataInterface() {propertyValues {FloatData [<0 0.05 0>, <0 -0.05 0>]}}
    )
    Use Rotation Vector=true
   ) {module=InitialMeshOrientation001}
   &CameraOffset(Camera Offset Amount=-5)
  ]
  particleUpdateScript [
   &SkeletalMeshLocation(
    Bone / Socket Index=&SelectIntFromArray(Direct Array Index=$Particles.PartitionParticles.Partition, Int Selection Array=DataInterface() {propertyValues {IntData [2,3]}})
    Bone Sampling Mode=::NewEnumerator4
    MaskSpawnGroup=false
    Orientation Sampling=::NewEnumerator1
    Sampled Position Offset=&SelectVectorFromArray(
     Direct Array Index=$Particles.PartitionParticles.Partition
     Vector Selection Array=DataInterface() {propertyValues {FloatData [<0 0 0>, <0 0 0>]}}
    )
    Skeletal Mesh=$System.Skeletal Mesh
   )
   &ParticleState(Kill Particles When Lifetime Has Elapsed=false, Let Infinitely Lived Particles Die When Emitter Deactivates=true)
   &Timeline(
    Play Rate=&Lerp_Float(A=2, B=3, Alpha=&ClampFloat(Float=$User.ParaGlideLeanAlpha))
    Timeline End=10000000
   )
   &DynamicMaterialParameters(
    Index 0 Param 1=&Lerp_Float(A=1, B=1, Alpha=$User.ParaGlideLeanAlpha)
    Index 0 Param 2=$Particles.Timeline.Time
    Index 2 Param 1=0.7
    Module.Index0Param0Write=true
    Module.Index0Param1Write=true
    Module.Index0Param2Write=true
    Module.Index0Param3Write=true
    Module.Index2Param0Write=true
    Module.Index2Param1Write=true
    Module.Index2Param2Write=true
    Module.Index2Param3Write=true
    Write Parameter Index 2=true
   )
   &ScaleMeshSize(Scale Factor=&Multiply_Vector(
     A=<1 1 1.3>
     B=&MakeVector(X=1, Y=1, Z=&Lerp_Float(A=1, B=2, Alpha=&ClampFloat(Float=$User.ParaGlideLeanAlpha)))
    ))
   &ScaleMeshSize(Scale Factor=&FN_FrontEndVector3Modifier(Gameplay=<1 1 1>, Front End=<0.6 0.6 0.9>)) {module=ScaleMeshSize001}
   &ScaleColor(
    Scale RGB=&Multiply_VectorByFloat(
     Vector=&MakeVectorFromLinearColorRGB(Color=&FNTimeOfDayColor(Night Color=<rgba 0.1444 0.6 0.3956 1>, Morning Color=<rgba 1 1 1 1>, Day Color=<rgba 1 1 1 1>, Evening Color=<rgba 1 1 1 1>))
     Float=&FN_FrontEndFloatModifier(
      Gameplay=&FN_PickFloatDependOnPlatform(TargetPlatformActive=1, TargetPlatformNoActive=0.8)
      Front End=1
     )
    )
    ScaleA=true
    ScaleRGB=true
   )
   /SetVariables(Particles.Color1:LinearColor=&MakeLinearColorFromVectorAndFloat(
     "Vector (RGB)"=&MakeVectorFromLinearColorRGB(Color=&MultiplyLinearColor(
       Linear Color=<rgba 1 1 1 1>
       Scale Factor=&MakeVector4_FromFloat(Float=&FN_FrontEndFloatModifier(
         Gameplay=&Multiply_Float(
          A=&Multiply_Float(
           A=$System.Color1 TOD fix
           B=&FNTimeOfDayScalar(Night Scale=0.2, Morning Scale=1, Evening Scale=1)
          )
          B=&FN_PickFloatDependOnPlatform(TargetPlatformActive=1, TargetPlatformNoActive=0.7)
         )
         Front End=0.7
        ))
      ))
     "Float (Alpha)"=&FN_FrontEndFloatModifier(Gameplay=5, Front End=2)
    )) {isSetParameters=true, module=SetVariables_#5}
   /SetVariables(Particles.Color2=&MakeLinearColorFromVectorAndFloat(
     "Vector (RGB)"=&MakeVectorFromLinearColorRGB(Color=&MultiplyLinearColor(
       Linear Color=&FNTimeOfDayColor(Night Color=<rgba 2 2 2 1>, Morning Color=<rgba 1 1 1 1>, Day Color=<rgba 1 1 1 1>, Evening Color=<rgba 1 1 1 1>)
       Scale Factor=&MakeVector4_FromFloat(Float=&FN_FrontEndFloatModifier(Gameplay=$System.Color2 TOD fix, Front End=1.1))
      ))
     "Float (Alpha)"=0.35
    )) {isSetParameters=true, module=SetVariables_#6}
   /SetVariables(Particles.Color3:LinearColor=&MakeLinearColorFromVectorAndFloat(
     "Vector (RGB)"=&MakeVectorFromLinearColorRGB(Color=<rgba 0.5 0.5 0.5 1>)
     "Float (Alpha)"=&FN_FrontEndFloatModifier(Gameplay=0.3, Front End=0.3)
    )) {isSetParameters=true, module=SetVariables_#7}
   &SolveForcesAndVelocity(Clamp Velocity=false, Limit Acceleration=false)
  ]
  rendererClasses [&NiagaraMeshRendererProperties]
  renderers [
   {rendererClass=&NiagaraMeshRendererProperties, rendererIndex=0}
   {rendererClass=&NiagaraMeshRendererProperties, rendererIndex=1}
   {rendererClass=&NiagaraMeshRendererProperties, rendererIndex=2}
   {rendererClass=&NiagaraMeshRendererProperties, rendererIndex=3}
  ]
 }
 Trail_hand {
  emitterUpdateScript [
   &EmitterState(
    Inactive Response=::NewEnumerator1
    Life Cycle Mode=::NewEnumerator1
    Loop Behavior=::NewEnumerator1
    Loop Duration Mode=::NewEnumerator1
    UseLoopDelay=false
   )
   &SpawnBurst_Instantaneous(Spawn Count=2, Use Spawn Probability=false, UseLoopCountLimit=false)
  ]
  particleSpawnScript [
   &PartitionParticles()
   &InitializeParticle(
    Color=<rgba 0.4 0.4 0.4 1>
    Mesh Scale=&SelectVectorFromArray(
     Direct Array Index=$Particles.PartitionParticles.Partition
     Vector Selection Array=DataInterface() {propertyValues {FloatData [<0.5 0.5 1.1>, <0.5 0.5 1.1>]}}
    )
    Mesh Scale Mode=::NewEnumerator4
    UsePositionOffset=false
    Write Material Random=true
   )
   &InitialMeshOrientation(
    Mesh Orientation Mode=::NewEnumerator2
    Rotation=&SelectVectorFromArray(
     Direct Array Index=$Particles.PartitionParticles.Partition
     Vector Selection Array=DataInterface() {propertyValues {FloatData [<0 0 -0.5>, <0 0 0>]}}
    )
    Use Rotation Vector=true
   )
   &InitialMeshOrientation(
    Mesh Orientation Mode=::NewEnumerator2
    Rotation=&MakeCustomVectorFromBool(
     Bool=$User.bIsFrontEndPreview
     True Vector=&SelectVectorFromArray(
      Direct Array Index=$Particles.PartitionParticles.Partition
      Vector Selection Array=DataInterface() {propertyValues {FloatData [<0 0 0>, <0 0 0>]}}
     )
     False Vector=&SelectVectorFromArray(
      Direct Array Index=$Particles.PartitionParticles.Partition
      Vector Selection Array=DataInterface() {propertyValues {FloatData [<-0.05 0.04 0>, <-0.05 -0.06 0>]}}
     )
    )
    Use Rotation Vector=true
   ) {module=InitialMeshOrientation001}
  ]
  particleUpdateScript [
   &SkeletalMeshLocation(
    Bone / Socket Index=$Particles.PartitionParticles.Partition
    Bone Sampling Mode=::NewEnumerator4
    MaskSpawnGroup=false
    Orientation Sampling=::NewEnumerator1
    Sampled Position Offset=&MakeCustomVectorFromBool(
     Bool=$User.bIsFrontEndPreview
     True Vector=&SelectVectorFromArray(
      Direct Array Index=$Particles.PartitionParticles.Partition
      Vector Selection Array=DataInterface() {propertyValues {FloatData [<0 5 0>, <0 -5 5>]}}
     )
     False Vector=&SelectVectorFromArray(
      Direct Array Index=$Particles.PartitionParticles.Partition
      Vector Selection Array=DataInterface() {propertyValues {FloatData [<0 0 0>, <0 0 5>]}}
     )
    )
    Skeletal Mesh=$System.Skeletal Mesh
   )
   &ParticleState(Kill Particles When Lifetime Has Elapsed=false, Let Infinitely Lived Particles Die When Emitter Deactivates=true)
   &Timeline(
    Play Rate=&Lerp_Float(A=2, B=3, Alpha=&ClampFloat(Float=$User.ParaGlideLeanAlpha))
    Timeline End=10000000
   )
   &DynamicMaterialParameters(
    Index 0 Param 1=&MakeCustomFloatFromBool(
     Bool=$User.bIsFrontEndPreview
     True Float=&Lerp_Float(A=0.9, B=1.3, Alpha=$User.ParaGlideLeanAlpha)
     False Float=&Lerp_Float(A=1.3, B=2.2, Alpha=$User.ParaGlideLeanAlpha)
    )
    Index 0 Param 2=$Particles.Timeline.Time
    Index 0 Param 3=&Lerp_Float(
     A=&FN_FrontEndFloatModifier(Gameplay=7, Front End=5)
     B=0
     Alpha=&ClampFloat(Float=$User.ParaGlideLeanAlpha)
    )
    Index 2 Param 1=0.7
    Module.Index0Param0Write=true
    Module.Index0Param1Write=true
    Module.Index0Param2Write=true
    Module.Index0Param3Write=true
    Module.Index2Param0Write=true
    Module.Index2Param1Write=true
    Module.Index2Param2Write=true
    Module.Index2Param3Write=true
    Write Parameter Index 2=true
   )
   &ScaleMeshSize(Scale Factor=&Multiply_Vector(
     A=&Multiply_Vector(A=<1 1 1.3>, B=&Lerp_Vector(A=<1 1 1>, B=<2 2 1>, Alpha=$User.ParaGlideLeanAlpha))
     B=&MakeVector(
      X=&Lerp_Float(A=1, B=1.3, Alpha=&ClampFloat(Float=$User.ParaGlideLeanAlpha))
      Y=&Lerp_Float(A=1, B=1.3, Alpha=&ClampFloat(Float=$User.ParaGlideLeanAlpha))
      Z=&Lerp_Float(A=0.8, B=1.7, Alpha=&ClampFloat(Float=$User.ParaGlideLeanAlpha))
     )
    ))
   &ScaleMeshSize(Scale Factor=&FN_FrontEndVector3Modifier(Gameplay=<1 1 1>, Front End=<0.7 0.7 0.8>)) {module=ScaleMeshSize001}
   &ScaleColor(
    Scale RGB=&Multiply_VectorByFloat(
     Vector=&MakeVectorFromLinearColorRGB(Color=&FNTimeOfDayColor(Night Color=<rgba 0.1444 0.6 0.3956 1>, Morning Color=<rgba 1 1 1 1>, Day Color=<rgba 1 1 1 1>, Evening Color=<rgba 1 1 1 1>))
     Float=&FN_FrontEndFloatModifier(
      Gameplay=&FN_PickFloatDependOnPlatform(TargetPlatformActive=1, TargetPlatformNoActive=0.8)
      Front End=1
     )
    )
    ScaleA=true
    ScaleRGB=true
   )
   /SetVariables(Particles.Color1:LinearColor=&MakeLinearColorFromVectorAndFloat(
     "Vector (RGB)"=&MakeVectorFromLinearColorRGB(Color=&MultiplyLinearColor(
       Linear Color=<rgba 1 1 1 1>
       Scale Factor=&MakeVector4_FromFloat(Float=&FN_FrontEndFloatModifier(
         Gameplay=&Multiply_Float(
          A=&Multiply_Float(
           A=$System.Color1 TOD fix
           B=&FNTimeOfDayScalar(Night Scale=0.2, Morning Scale=1, Evening Scale=1)
          )
          B=&FN_PickFloatDependOnPlatform(TargetPlatformActive=1, TargetPlatformNoActive=0.7)
         )
         Front End=0.7
        ))
      ))
     "Float (Alpha)"=&FN_FrontEndFloatModifier(Gameplay=5, Front End=2)
    )) {isSetParameters=true, module=SetVariables_#5}
   /SetVariables(Particles.Color2=&MakeLinearColorFromVectorAndFloat(
     "Vector (RGB)"=&MakeVectorFromLinearColorRGB(Color=&MultiplyLinearColor(
       Linear Color=&FNTimeOfDayColor(Night Color=<rgba 2 2 2 1>, Morning Color=<rgba 1 1 1 1>, Day Color=<rgba 1 1 1 1>, Evening Color=<rgba 1 1 1 1>)
       Scale Factor=&MakeVector4_FromFloat(Float=&FN_FrontEndFloatModifier(Gameplay=$System.Color2 TOD fix, Front End=1.1))
      ))
     "Float (Alpha)"=0.35
    )) {isSetParameters=true, module=SetVariables_#6}
   /SetVariables(Particles.Color3:LinearColor=&MakeLinearColorFromVectorAndFloat(
     "Vector (RGB)"=&MakeVectorFromLinearColorRGB(Color=<rgba 0.5 0.5 0.5 1>)
     "Float (Alpha)"=&FN_FrontEndFloatModifier(Gameplay=0.3, Front End=0.3)
    )) {isSetParameters=true, module=SetVariables_#7}
  ]
  rendererClasses [&NiagaraMeshRendererProperties]
  renderers [
   {rendererClass=&NiagaraMeshRendererProperties, rendererIndex=0}
   {rendererClass=&NiagaraMeshRendererProperties, rendererIndex=1}
   {rendererClass=&NiagaraMeshRendererProperties, rendererIndex=2}
   {rendererClass=&NiagaraMeshRendererProperties, rendererIndex=3}
  ]
 }
}
summary {
 systemName=NS_ConTrail_MadameMoth_Emerald
 emitters [
  {emitterName=Trail_hand, rendererClasses [&NiagaraMeshRendererProperties]}
  {emitterName=Trail_feet, rendererClasses [&NiagaraMeshRendererProperties]}
  {emitterName=Trail_Ring, rendererClasses [&NiagaraMeshRendererProperties]}
  {emitterName=ParticleDecal, rendererClasses [&NiagaraSpriteRendererProperties, &NiagaraLightRendererProperties]}
  {emitterName=Sparks, simTarget=GPUComputeSim, rendererClasses [&NiagaraSpriteRendererProperties]}
  {emitterName=Fire001, simTarget=GPUComputeSim, rendererClasses [&NiagaraSpriteRendererProperties]}
 ]
 userVariables [
  User.bIsFrontEnd {type=NiagaraBool}
  User.bIsFrontEndPreview {type=NiagaraBool}
  User.ParaGlideLeanAlpha {type=NiagaraFloat, defaultValue=0, description=""}
 ]
}
